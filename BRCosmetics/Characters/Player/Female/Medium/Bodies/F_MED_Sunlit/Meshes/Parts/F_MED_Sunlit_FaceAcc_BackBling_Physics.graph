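# Physics asset for the F_MED_Sunlit face accessory, BackBling variant, giving its left and right braid bone chains collision and joint limits.
# Capsules on spine, neck, head and clavicles form the body collision, including an oversized offset capsule on the right clavicle.
# The left braid chains directly from the head with mostly tight swing limits.
# The right braid hangs from a free-moving master bone, with wider swing limits down its chain.
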PhysicsAsset /BRCosmetics/Characters/Player/Female/Medium/Bodies/F_MED_Sunlit/Meshes/Parts/F_MED_Sunlit_FaceAcc_BackBling_Physics
bodies {
 clavicle_l [
  {
   shapeType=Capsule
   length=9.6391
   radius=8
   shapeName=clavicle_l_capsule
   center=<1.8624 0.0004 -2.1272>
   rotation=<rot 90 -179.5524 175.4437>
  }
  {shapeType=Capsule, length=30, radius=6, shapeName=clavicle_l_capsule, center=<15 0 13>, rotation=<rot -20 0 4>}
 ]
 clavicle_r [
  {shapeType=Capsule, length=29.2815, radius=25, shapeName=clavicle_l_capsule, center=<-37.2423 0 -10.1754>, rotation=<rot 0 0 90>}
  {shapeType=Capsule, length=30, radius=7, shapeName=clavicle_r_capsule, center=<-15 0 -13>, rotation=<rot -20 0 -4>}
 ]
 dyn_braid_l_04 [{shapeType=Capsule, length=5, radius=1.4, shapeName=dyn_braid_l_04_capsule, rotation=<rot 90 0 0>}]
 dyn_braid_l_05 [{shapeType=Capsule, length=5, radius=1.4, shapeName=dyn_braid_l_05_capsule, rotation=<rot 90 0 0>}]
 dyn_braid_l_06 [{shapeType=Capsule, length=5, radius=1.4, shapeName=dyn_braid_l_06_capsule, rotation=<rot 90 0 0>}]
 dyn_braid_l_07 [{shapeType=Capsule, length=5, radius=1.4, shapeName=dyn_braid_l_07_capsule, rotation=<rot 90 0 0>}]
 dyn_braid_l_08 [{shapeType=Capsule, length=5, radius=1.4, shapeName=dyn_braid_l_08_capsule, rotation=<rot 90 0 0>}]
 dyn_braid_l_09 [{shapeType=Capsule, length=5, radius=1.4, shapeName=dyn_braid_l_09_capsule, rotation=<rot 90 0 0>}]
 dyn_braid_r_02 [{shapeType=Capsule, length=3, radius=1.5, shapeName=dyn_braid_r_02_capsule, center=<2.9007 0 0>, rotation=<rot 90 0 0>}]
 dyn_braid_r_03 [{shapeType=Capsule, length=1.592, radius=2.5, shapeName=dyn_braid_r_03_capsule, center=<1 0 0>, rotation=<rot 90 0 0>}]
 dyn_braid_r_04 [{shapeType=Capsule, length=3.25, radius=1.15, shapeName=dyn_braid_r_04_capsule, center=<1.1637 0 0>, rotation=<rot 90 0 0>}]
 dyn_braid_r_05 [{shapeType=Capsule, length=3.25, radius=1.15, shapeName=dyn_braid_r_05_capsule, center=<2 0 0>, rotation=<rot 90 0 0>}]
 dyn_braid_r_06 [{shapeType=Capsule, length=3.25, radius=1.15, shapeName=dyn_braid_r_06_capsule, center=<2 0 0>, rotation=<rot 90 0 0>}]
 dyn_braid_r_07 [{shapeType=Capsule, length=3.75, radius=1.65, shapeName=dyn_braid_r_07_capsule, center=<2 0 0>, rotation=<rot 90 0 0>}]
 dyn_braid_r_MSTR [{shapeType=Capsule, length=0.505, radius=0.505, shapeName=dyn_braid_r_MSTR_capsule, rotation=<rot 0 -90 0>}]
 head [{shapeType=Capsule, length=4.5, radius=7, center=<2.1752 1.6556 0>, rotation=<rot 90 0 0>}]
 neck_01 [{shapeType=Capsule, length=13.75, radius=4.5, shapeName=neck_02_capsule, center=<-4.1691 -0.7987 3.8326>}]
 neck_02 [{shapeType=Capsule, length=6.25, radius=4.25, shapeName=neck_02_capsule, center=<-2.461 0 0>, rotation=<rot 90 0 0>}]
 spine_04 [{shapeType=Capsule, length=14.25, radius=10.25, shapeName=spine_05_capsule, center=<-6.1728 1.5108 0>, rotation=<rot -90 0 -0.0001>}]
 spine_05 [{shapeType=Capsule, length=12.75, radius=8.25, shapeName=spine_05_capsule}]
}
constraints [
 {
  bone1Name=dyn_braid_r_MSTR
  bone2Name=head
  swing1LimitDegrees=45
  swing1Motion=Free
  swing2LimitDegrees=45
  swing2Motion=Free
  twistLimitDegrees=45
  twistMotion=Free
 }
 {bone1Name=dyn_braid_r_03, bone2Name=dyn_braid_r_02, swing1LimitDegrees=40, swing2LimitDegrees=15, twistLimitDegrees=0}
 {bone1Name=dyn_braid_r_04, bone2Name=dyn_braid_r_03, swing1LimitDegrees=40, swing2LimitDegrees=40, twistLimitDegrees=0}
 {bone1Name=dyn_braid_r_05, bone2Name=dyn_braid_r_04, swing1LimitDegrees=30, swing2LimitDegrees=40, twistLimitDegrees=0}
 {bone1Name=dyn_braid_r_06, bone2Name=dyn_braid_r_05, swing1LimitDegrees=30, swing2LimitDegrees=40, twistLimitDegrees=0}
 {bone1Name=dyn_braid_r_07, bone2Name=dyn_braid_r_06, swing1LimitDegrees=30, swing2LimitDegrees=30, twistLimitDegrees=0}
 {
  bone1Name=head
  bone2Name=neck_02
  swing1LimitDegrees=45
  swing1Motion=Free
  swing2LimitDegrees=45
  swing2Motion=Free
  twistLimitDegrees=45
  twistMotion=Free
 }
 {
  bone1Name=spine_05
  bone2Name=spine_04
  swing1LimitDegrees=45
  swing1Motion=Free
  swing2LimitDegrees=45
  swing2Motion=Free
  twistLimitDegrees=45
  twistMotion=Free
 }
 {
  bone1Name=clavicle_l
  bone2Name=spine_05
  swing1LimitDegrees=45
  swing1Motion=Free
  swing2LimitDegrees=45
  swing2Motion=Free
  twistLimitDegrees=45
  twistMotion=Free
 }
 {
  bone1Name=clavicle_r
  bone2Name=spine_05
  swing1LimitDegrees=45
  swing1Motion=Free
  swing2LimitDegrees=45
  swing2Motion=Free
  twistLimitDegrees=45
  twistMotion=Free
 }
 {
  bone1Name=neck_02
  bone2Name=neck_01
  swing1LimitDegrees=45
  swing1Motion=Free
  swing2LimitDegrees=45
  swing2Motion=Free
  twistLimitDegrees=45
  twistMotion=Free
 }
 {
  bone1Name=neck_01
  bone2Name=spine_05
  swing1LimitDegrees=45
  swing1Motion=Free
  swing2LimitDegrees=45
  swing2Motion=Free
  twistLimitDegrees=45
  twistMotion=Free
 }
 {
  bone1Name=dyn_braid_r_02
  bone2Name=dyn_braid_r_MSTR
  swing1LimitDegrees=180
  swing1Motion=Free
  swing2LimitDegrees=180
  swing2Motion=Free
  twistLimitDegrees=0
  twistMotion=Free
 }
 {bone1Name=dyn_braid_l_04, bone2Name=head, swing1LimitDegrees=40, swing2LimitDegrees=60, twistLimitDegrees=0}
 {bone1Name=dyn_braid_l_05, bone2Name=dyn_braid_l_04, swing1LimitDegrees=15, swing2LimitDegrees=15, twistLimitDegrees=0}
 {bone1Name=dyn_braid_l_06, bone2Name=dyn_braid_l_05, swing1LimitDegrees=15, swing2LimitDegrees=15, twistLimitDegrees=0}
 {bone1Name=dyn_braid_l_07, bone2Name=dyn_braid_l_06, swing1LimitDegrees=15, swing2LimitDegrees=15, twistLimitDegrees=0}
 {bone1Name=dyn_braid_l_08, bone2Name=dyn_braid_l_07, swing1LimitDegrees=15, swing2LimitDegrees=15, twistLimitDegrees=0}
 {bone1Name=dyn_braid_l_09, bone2Name=dyn_braid_l_08, swing1LimitDegrees=15, swing2LimitDegrees=15, twistLimitDegrees=0}
]
