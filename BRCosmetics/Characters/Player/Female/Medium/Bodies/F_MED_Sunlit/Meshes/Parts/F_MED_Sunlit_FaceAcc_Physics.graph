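# Physics asset for the F_MED_Sunlit female character's FaceAcc mesh, defining capsule bodies for two dynamic braids and the upper body around them.
# Each braid chains from a small master body on the head, the left longer than the right, with limited swing and twist along its segments.
# The right braid's root joint rotates freely, while the left's is tightly limited.
# Head, neck, upper spine, clavicle and upper arm capsules are linked by free-rotation constraints.
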PhysicsAsset /BRCosmetics/Characters/Player/Female/Medium/Bodies/F_MED_Sunlit/Meshes/Parts/F_MED_Sunlit_FaceAcc_Physics
bodies {
 clavicle_l [{shapeType=Capsule, length=4.4567, radius=5.3263, shapeName=clavicle_l_capsule, center=<4.8357 0 0>, rotation=<rot 90 0 0>}]
 clavicle_r [{shapeType=Capsule, length=5.1662, radius=4.6168, shapeName=clavicle_l_capsule, center=<-4.4055 0.2855 0>, rotation=<rot 90 0 0>}]
 dyn_braid_l_02 [{shapeType=Capsule, length=0.8384, radius=1, shapeName=dyn_braid_l_02_capsule, rotation=<rot 90 0 0>}]
 dyn_braid_l_03 [{shapeType=Capsule, length=5.0884, radius=3.25, shapeName=dyn_backBraid_1_capsule, center=<2.5 0 0>, rotation=<rot 75 89.9998 79.9997>}]
 dyn_braid_l_04 [{shapeType=Capsule, length=2.3995, radius=3.5389, shapeName=dyn_braid_l_04_capsule, center=<2.5 0 1>, rotation=<rot 90 0 0>}]
 dyn_braid_l_05 [{shapeType=Capsule, length=2.3995, radius=3.5389, shapeName=dyn_braid_l_05_capsule, center=<2.5 0 1>, rotation=<rot 90 0 0>}]
 dyn_braid_l_06 [{shapeType=Capsule, length=2.3995, radius=3.5389, shapeName=dyn_braid_l_06_capsule, center=<2.5 0 1>, rotation=<rot 90 0 0>}]
 dyn_braid_l_07 [{shapeType=Capsule, length=2.3995, radius=3.5389, shapeName=dyn_braid_l_07_capsule, center=<2.5 0 1>, rotation=<rot 90 0 0>}]
 dyn_braid_l_08 [{shapeType=Capsule, length=2.3995, radius=3.5389, shapeName=dyn_braid_l_08_capsule, center=<2.5 0 1>, rotation=<rot 90 0 0>}]
 dyn_braid_l_09 [{shapeType=Capsule, length=1.8995, radius=4.0389, shapeName=dyn_braid_l_09_capsule, center=<2.5 0 1>, rotation=<rot 90 0 0>}]
 dyn_braid_l_MSTR [{shapeType=Capsule, length=0.505, radius=0.505, shapeName=dyn_braid_l_MSTR_capsule, rotation=<rot 0 -90 0>}]
 dyn_braid_r_02 [{shapeType=Capsule, length=3, radius=1.5, shapeName=dyn_braid_r_02_capsule, center=<2.9007 0 0>, rotation=<rot 90 0 0>}]
 dyn_braid_r_03 [{shapeType=Capsule, length=1.4304, radius=2.5, shapeName=dyn_braid_r_03_capsule, center=<1 0 0>, rotation=<rot 90 0 0>}]
 dyn_braid_r_04 [{shapeType=Capsule, length=1.7943, radius=2.444, shapeName=dyn_braid_r_04_capsule, center=<1.1637 0 1>, rotation=<rot 90 0 0>}]
 dyn_braid_r_05 [{shapeType=Capsule, length=2.0826, radius=2.1557, shapeName=dyn_braid_r_05_capsule, center=<2 0 1>, rotation=<rot 90 0 0>}]
 dyn_braid_r_06 [{shapeType=Capsule, length=2.5834, radius=1.655, shapeName=dyn_braid_r_06_capsule, center=<2 0 1>, rotation=<rot 90 0 0>}]
 dyn_braid_r_07 [{shapeType=Capsule, length=3.5884, radius=1.65, shapeName=dyn_braid_r_07_capsule, center=<2 0 1>, rotation=<rot 90 0 0>}]
 dyn_braid_r_MSTR [{shapeType=Capsule, length=0.505, radius=0.505, shapeName=dyn_braid_r_MSTR_capsule, rotation=<rot 0 -90 0>}]
 head [{shapeType=Capsule, length=4.5, radius=7, center=<2.1752 1.6556 0>, rotation=<rot 90 0 0>}]
 neck_01 [{shapeType=Capsule, length=2.8068, radius=4.5, shapeName=neck_02_capsule, rotation=<rot 90 0 0>}]
 neck_02 [{shapeType=Capsule, length=6.25, radius=4.25, shapeName=neck_02_capsule, rotation=<rot 90 0 0>}]
 spine_04 [{shapeType=Capsule, length=6.2272, radius=9.2409, shapeName=spine_05_capsule, center=<0 2.1973 0>}]
 spine_05 [{shapeType=Capsule, length=7.6358, radius=7.6336, shapeName=spine_05_capsule}]
 upperarm_l [{shapeType=Capsule, length=18.1303, radius=4.8697, shapeName=upperarm_l_capsule, center=<11.921 0 0>, rotation=<rot 90 0 0>}]
 upperarm_r [{shapeType=Capsule, length=18.1303, radius=4.8697, shapeName=upperarm_r_capsule, center=<-11.679 0 0>, rotation=<rot 90 0 0>}]
}
constraints [
 {bone1Name=dyn_braid_l_03, bone2Name=dyn_braid_l_02, swing1LimitDegrees=20, swing2LimitDegrees=55, twistLimitDegrees=0}
 {bone1Name=dyn_braid_l_04, bone2Name=dyn_braid_l_03, swing1LimitDegrees=20, swing2LimitDegrees=60, twistLimitDegrees=5}
 {bone1Name=dyn_braid_l_05, bone2Name=dyn_braid_l_04, swing1LimitDegrees=20, swing2LimitDegrees=30, twistLimitDegrees=5}
 {bone1Name=dyn_braid_l_06, bone2Name=dyn_braid_l_05, swing1LimitDegrees=30, swing2LimitDegrees=30, twistLimitDegrees=10}
 {bone1Name=dyn_braid_l_07, bone2Name=dyn_braid_l_06, swing1LimitDegrees=30, swing2LimitDegrees=30, twistLimitDegrees=10}
 {bone1Name=dyn_braid_l_08, bone2Name=dyn_braid_l_07, swing1LimitDegrees=30, swing2LimitDegrees=30, twistLimitDegrees=10}
 {bone1Name=dyn_braid_l_09, bone2Name=dyn_braid_l_08, swing1LimitDegrees=30, swing2LimitDegrees=30, twistLimitDegrees=10}
 {
  bone1Name=dyn_braid_l_MSTR
  bone2Name=head
  swing1LimitDegrees=45
  swing1Motion=Free
  swing2LimitDegrees=45
  swing2Motion=Free
  twistLimitDegrees=45
  twistMotion=Free
 }
 {
  bone1Name=dyn_braid_r_MSTR
  bone2Name=head
  swing1LimitDegrees=45
  swing1Motion=Free
  swing2LimitDegrees=45
  swing2Motion=Free
  twistLimitDegrees=45
  twistMotion=Free
 }
 {bone1Name=dyn_braid_r_03, bone2Name=dyn_braid_r_02, swing1LimitDegrees=15, swing2LimitDegrees=15, twistLimitDegrees=5}
 {bone1Name=dyn_braid_r_04, bone2Name=dyn_braid_r_03, swing1LimitDegrees=30, swing2LimitDegrees=30, twistLimitDegrees=10}
 {bone1Name=dyn_braid_r_05, bone2Name=dyn_braid_r_04, swing1LimitDegrees=30, swing2LimitDegrees=30, twistLimitDegrees=10}
 {bone1Name=dyn_braid_r_06, bone2Name=dyn_braid_r_05, swing1LimitDegrees=30, swing2LimitDegrees=30, twistLimitDegrees=10}
 {bone1Name=dyn_braid_r_07, bone2Name=dyn_braid_r_06, swing1LimitDegrees=30, swing2LimitDegrees=30, twistLimitDegrees=10}
 {
  bone1Name=head
  bone2Name=neck_02
  swing1LimitDegrees=45
  swing1Motion=Free
  swing2LimitDegrees=45
  swing2Motion=Free
  twistLimitDegrees=45
  twistMotion=Free
 }
 {
  bone1Name=spine_05
  bone2Name=spine_04
  swing1LimitDegrees=45
  swing1Motion=Free
  swing2LimitDegrees=45
  swing2Motion=Free
  twistLimitDegrees=45
  twistMotion=Free
 }
 {
  bone1Name=clavicle_l
  bone2Name=spine_05
  swing1LimitDegrees=45
  swing1Motion=Free
  swing2LimitDegrees=45
  swing2Motion=Free
  twistLimitDegrees=45
  twistMotion=Free
 }
 {
  bone1Name=clavicle_r
  bone2Name=spine_05
  swing1LimitDegrees=45
  swing1Motion=Free
  swing2LimitDegrees=45
  swing2Motion=Free
  twistLimitDegrees=45
  twistMotion=Free
 }
 {
  bone1Name=neck_02
  bone2Name=neck_01
  swing1LimitDegrees=45
  swing1Motion=Free
  swing2LimitDegrees=45
  swing2Motion=Free
  twistLimitDegrees=45
  twistMotion=Free
 }
 {
  bone1Name=neck_01
  bone2Name=spine_05
  swing1LimitDegrees=45
  swing1Motion=Free
  swing2LimitDegrees=45
  swing2Motion=Free
  twistLimitDegrees=45
  twistMotion=Free
 }
 {
  bone1Name=dyn_braid_r_02
  bone2Name=dyn_braid_r_MSTR
  swing1LimitDegrees=180
  swing1Motion=Free
  swing2LimitDegrees=180
  swing2Motion=Free
  twistLimitDegrees=0
  twistMotion=Free
 }
 {bone1Name=dyn_braid_l_02, bone2Name=dyn_braid_l_MSTR, swing1LimitDegrees=10, swing2LimitDegrees=10, twistLimitDegrees=0}
 {
  bone1Name=upperarm_l
  bone2Name=clavicle_l
  swing1LimitDegrees=45
  swing1Motion=Free
  swing2LimitDegrees=45
  swing2Motion=Free
  twistLimitDegrees=45
  twistMotion=Free
 }
 {
  bone1Name=upperarm_r
  bone2Name=clavicle_r
  swing1LimitDegrees=45
  swing1Motion=Free
  swing2LimitDegrees=45
  swing2Motion=Free
  twistLimitDegrees=45
  twistMotion=Free
 }
]
