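# Physics asset for the Sunlit female medium character's face accessory mesh, a skydive variant that gives the two hair braids physical motion.
# Capsule bodies form a longer left and shorter right braid chain, both hanging from the head, with twist locked and tighter swing limits at the roots.
# Capsules on the head, neck, upper spine, clavicles and upper arms, joined by free-motion constraints, supply body volumes around the braids.
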PhysicsAsset /BRCosmetics/Characters/Player/Female/Medium/Bodies/F_MED_Sunlit/Meshes/Parts/F_MED_Sunlit_FaceAcc_Physics_Skydive
bodies {
 clavicle_l [{shapeType=Capsule, length=9.6391, radius=8, shapeName=clavicle_l_capsule, center=<2 0 -2>, rotation=<rot 90 -179.5524 175.4437>}]
 clavicle_r [{shapeType=Capsule, length=9.6391, radius=7, shapeName=clavicle_l_capsule, center=<-2 1 2>, rotation=<rot 90 -179.5524 175.4437>}]
 dyn_braid_l_03 [{shapeType=Capsule, length=5, radius=1.4, shapeName=dyn_braid_l_04_capsule, center=<1 0 0>, rotation=<rot 90 0 0>}]
 dyn_braid_l_04 [{shapeType=Capsule, length=5, radius=1.4, shapeName=dyn_braid_l_04_capsule, center=<1 0 0>, rotation=<rot 90 0 0>}]
 dyn_braid_l_05 [{shapeType=Capsule, length=5, radius=1.4, shapeName=dyn_braid_l_05_capsule, center=<1 0 0>, rotation=<rot 90 0 0>}]
 dyn_braid_l_06 [{shapeType=Capsule, length=5, radius=1.4, shapeName=dyn_braid_l_06_capsule, center=<1 0 0>, rotation=<rot 90 0 0>}]
 dyn_braid_l_07 [{shapeType=Capsule, length=5, radius=1.4, shapeName=dyn_braid_l_07_capsule, center=<1 0 0>, rotation=<rot 90 0 0>}]
 dyn_braid_l_08 [{shapeType=Capsule, length=5, radius=1.4, shapeName=dyn_braid_l_08_capsule, center=<1 0 0>, rotation=<rot 90 0 0>}]
 dyn_braid_l_09 [{shapeType=Capsule, length=5, radius=1.4, shapeName=dyn_braid_l_09_capsule, center=<1 0 0>, rotation=<rot 90 0 0>}]
 dyn_braid_r_03 [{shapeType=Capsule, length=5, radius=2, shapeName=dyn_braid_r_03_capsule, center=<3 0 0>, rotation=<rot 90 0 0>}]
 dyn_braid_r_04 [{shapeType=Capsule, length=5, radius=1.4, shapeName=dyn_braid_r_04_capsule, center=<1 0 0>, rotation=<rot 90 0 0>}]
 dyn_braid_r_05 [{shapeType=Capsule, length=5, radius=1.4, shapeName=dyn_braid_r_05_capsule, center=<1 0 0>, rotation=<rot 90 0 0>}]
 dyn_braid_r_06 [{shapeType=Capsule, length=5, radius=1.4, shapeName=dyn_braid_r_06_capsule, center=<1 0 0>, rotation=<rot 90 0 0>}]
 dyn_braid_r_07 [{shapeType=Capsule, length=5, radius=1.7, shapeName=dyn_braid_r_07_capsule, center=<1 0 0>, rotation=<rot 90 0 0>}]
 head [{shapeType=Capsule, length=0.5, radius=0.5, shapeName=head_capsule, rotation=<rot 90 0 0>}]
 neck_01 [{shapeType=Capsule, length=15, radius=5, shapeName=neck_02_capsule, center=<-5.6553 0 0>}]
 neck_02 [{shapeType=Capsule, length=6, radius=5, shapeName=neck_02_capsule, rotation=<rot 90 0 0>}]
 spine_05 [{shapeType=Capsule, length=7, radius=9, shapeName=spine_05_capsule, center=<-2.2333 0.3463 0>}]
 upperarm_l [{shapeType=Capsule, length=20, radius=6, shapeName=upperarm_l_capsule, center=<14 0 -1>, rotation=<rot 90 0 0>}]
 upperarm_r [{shapeType=Capsule, length=20, radius=6, shapeName=upperarm_l_capsule, center=<-14 0 1>, rotation=<rot 90 0 0>}]
}
constraints [
 {
  bone1Name=clavicle_l
  bone2Name=spine_05
  swing1LimitDegrees=45
  swing1Motion=Free
  swing2LimitDegrees=45
  swing2Motion=Free
  twistLimitDegrees=45
  twistMotion=Free
 }
 {
  bone1Name=upperarm_r
  bone2Name=clavicle_r
  swing1LimitDegrees=45
  swing1Motion=Free
  swing2LimitDegrees=45
  swing2Motion=Free
  twistLimitDegrees=45
  twistMotion=Free
 }
 {
  bone1Name=neck_01
  bone2Name=spine_05
  swing1LimitDegrees=45
  swing1Motion=Free
  swing2LimitDegrees=45
  swing2Motion=Free
  twistLimitDegrees=45
  twistMotion=Free
 }
 {
  bone1Name=neck_02
  bone2Name=neck_01
  swing1LimitDegrees=45
  swing1Motion=Free
  swing2LimitDegrees=45
  swing2Motion=Free
  twistLimitDegrees=45
  twistMotion=Free
 }
 {
  bone1Name=head
  bone2Name=neck_02
  swing1LimitDegrees=45
  swing1Motion=Free
  swing2LimitDegrees=45
  swing2Motion=Free
  twistLimitDegrees=45
  twistMotion=Free
 }
 {bone1Name=dyn_braid_l_04, bone2Name=dyn_braid_l_03, swing1LimitDegrees=30, swing2LimitDegrees=30, twistLimitDegrees=45, twistMotion=Locked}
 {bone1Name=dyn_braid_l_05, bone2Name=dyn_braid_l_04, swing1LimitDegrees=30, swing2LimitDegrees=30, twistLimitDegrees=45, twistMotion=Locked}
 {bone1Name=dyn_braid_l_06, bone2Name=dyn_braid_l_05, swing1LimitDegrees=30, swing2LimitDegrees=30, twistLimitDegrees=45, twistMotion=Locked}
 {bone1Name=dyn_braid_l_07, bone2Name=dyn_braid_l_06, swing1LimitDegrees=30, swing2LimitDegrees=30, twistLimitDegrees=45, twistMotion=Locked}
 {bone1Name=dyn_braid_l_08, bone2Name=dyn_braid_l_07, swing1LimitDegrees=30, swing2LimitDegrees=30, twistLimitDegrees=45, twistMotion=Locked}
 {bone1Name=dyn_braid_l_09, bone2Name=dyn_braid_l_08, swing1LimitDegrees=30, swing2LimitDegrees=30, twistLimitDegrees=45, twistMotion=Locked}
 {bone1Name=dyn_braid_r_03, bone2Name=head, swing1LimitDegrees=20, swing2LimitDegrees=20, twistLimitDegrees=45, twistMotion=Locked}
 {bone1Name=dyn_braid_r_04, bone2Name=dyn_braid_r_03, swing1LimitDegrees=30, swing2LimitDegrees=30, twistLimitDegrees=45, twistMotion=Locked}
 {bone1Name=dyn_braid_r_05, bone2Name=dyn_braid_r_04, swing1LimitDegrees=30, swing2LimitDegrees=30, twistLimitDegrees=45, twistMotion=Locked}
 {bone1Name=dyn_braid_r_06, bone2Name=dyn_braid_r_05, swing1LimitDegrees=30, swing2LimitDegrees=30, twistLimitDegrees=45, twistMotion=Locked}
 {bone1Name=dyn_braid_r_07, bone2Name=dyn_braid_r_06, swing1LimitDegrees=30, swing2LimitDegrees=30, twistLimitDegrees=45, twistMotion=Locked}
 {
  bone1Name=clavicle_r
  bone2Name=spine_05
  swing1LimitDegrees=45
  swing1Motion=Free
  swing2LimitDegrees=45
  swing2Motion=Free
  twistLimitDegrees=45
  twistMotion=Free
 }
 {bone1Name=dyn_braid_l_03, bone2Name=head, swing1LimitDegrees=20, swing2LimitDegrees=20, twistLimitDegrees=45, twistMotion=Locked}
 {
  bone1Name=upperarm_l
  bone2Name=clavicle_l
  swing1LimitDegrees=45
  swing1Motion=Free
  swing2LimitDegrees=45
  swing2Motion=Free
  twistLimitDegrees=45
  twistMotion=Free
 }
]
